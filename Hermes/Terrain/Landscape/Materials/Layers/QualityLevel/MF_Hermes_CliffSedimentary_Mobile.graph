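# Mobile material-layer function for sedimentary cliffs in Hermes terrain, outputting base color, specular, roughness, normal and opacity mask.
# Wave-distorted world-space UVs drive biplanar XZ/YZ samples of one packed mask texture, yielding normal and detail channels.
# Color blends tinted layers and speckles, with repeating world-height bands adding variation.
# Static switches, fed by an IsStaticMesh input, select texture, blend sharpness, normal construction and height source.
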MaterialFunction /Hermes/Terrain/Landscape/Materials/Layers/QualityLevel/MF_Hermes_CliffSedimentary_Mobile
graph {nodes [
  ComponentMask() {g=true, r=true}
  n1 = SetMaterialAttributes!(
   MaterialAttributes
   BaseColor=NamedRerouteUsage()
   Specular=NamedRerouteUsage()
   Roughness=NamedRerouteUsage()
   Normal=NamedRerouteUsage()
   OpacityMask=NamedRerouteUsage()
  ) {
   desc=HighEnd Material
   attributeSetTypes [
    #1
    #2
    #3
    #4
    #5
   ]
  }
  FunctionOutput(None=@n1)
  n2 = TruncateLWC!(None=WorldPosition().XYZ)
  n3 = VectorParameter() {
   group=Cliff
   parameterName="Sediment_a | Pos && Scale"
   sortPriority=306
   channelNames {a=Scale, b=Z, g=Y, r=X}
   defaultValue=<rgba 0 0 0 2500>
  }
  n4 = VectorParameter() {
   group=Cliff
   parameterName="Sediment_a | Wave Dist Pos && Freq"
   sortPriority=307
   channelNames {a=Freq, b=Z, g=Y, r=X}
   defaultValue=<rgba 0 0 0 10000>
  }
  n5 = MaterialFunctionCall!(
   WorldPosition=@n2
   Sediment L2_WS | Position Offset_1=@n3
   Sediment L2_WS | Scale=@n3
   Use Wavy Distortion=StaticBool() {value=true}
   Wave Distortion Intensity=ScalarParameter() {defaultValue=2500, group=Cliff, parameterName=Sediment_a | Wave Dist Intensity, sortPriority=307}
   Sediment L2_WS | Wave DistPos Offset=@n4
   Wave Distortion Freq=@n4
  ) {materialFunction=/Game/Global/Base4Layers_2022/MaterialFunctions/MF_Cliffs_SedimentWS_UVs}
  n6 = TextureObjectParameter() {
   group=Mobile CliffSedimentary
   parameterName=Mobile CliffSedimentary Texture StaticMeshes
   samplerType=SAMPLERTYPE_Masks
   texture=/Game/Environments/Helios/Terrain/TexturesMobile/T_Helios_Mobile_CliffSedimentary_StaticMeshes
  }
  n7 = TextureObjectParameter() {
   group=Mobile CliffSedimentary
   parameterName=Mobile CliffSedimentary Texture
   samplerType=SAMPLERTYPE_Masks
   texture=/Game/Environments/Helios/Terrain/TexturesMobile/T_Helios_Mobile_CliffSedimentary
  }
  n8 = StaticSwitch(True=@n6, False=@n7, Value=NamedRerouteUsage())
  n9 = TextureSample(UVs=@n5.XZ, Tex=@n8) {samplerType=SAMPLERTYPE_Masks, texture=/Game/Environments/Helios/Terrain/TexturesMobile/T_Helios_Mobile_Cliff}
  n10 = TextureSample(UVs=@n5.YZ, Tex=@n8) {samplerType=SAMPLERTYPE_Masks, texture=&T_Helios_Mobile_Cliff}
  n11 = MaterialFunctionCall!(Sharpness=StaticSwitch(True=Constant() {r=1.5}, False=Constant() {r=10}, Value=NamedRerouteUsage())) {materialFunction=/Game/Athena/Helios/Environments/Landscape/Materials/MaterialFunctions_Utilities/MF_MobileCliffBiplanarMask}
  n12 = LinearInterpolate(A=AppendVector(A=@n9.R, B=@n9.R), B=AppendVector(A=@n10.R, B=@n10.R), Alpha=@n11.Result)
  n13 = DeriveNormalZ(InXY=Add(A=Multiply(A=@n12) {constB=2}) {constB=-1})
  n14 = ComponentMask(None=@n13) {r=true}
  n15 = VertexNormalWS()
  n16 = ComponentMask(None=@n15) {r=true}
  n17 = ComponentMask(None=@n13) {b=true}
  n18 = ComponentMask(None=@n15) {g=true}
  n19 = Multiply(A=ComponentMask(None=@n13) {g=true}) {constB=-1}
  n20 = ComponentMask(None=@n15) {b=true}
  n21 = LinearInterpolate(
   A=MaterialFunctionCall(R=Add(A=@n14, B=@n16), G=Multiply(A=@n17, B=@n18) {constB=-1}, B=Add(A=@n19, B=@n20)) {materialFunction=&AppendMany}.RGB
   B=MaterialFunctionCall(R=Multiply(A=@n17, B=@n16) {constB=-1}, G=Add(A=@n14, B=@n18), B=Add(A=@n19, B=@n20)) {materialFunction=&AppendMany}.RGB
   Alpha=@n11.Result
  )
  n22 = Multiply(A=VertexNormalWS()) {constB=-1}
  n23 = ComponentMask(None=@n22) {g=true}
  n24 = Subtract(A=Saturate(None=Sign(None=@n23)), B=@n9.G)
  n25 = ComponentMask(None=@n22) {r=true}
  n26 = Subtract(A=Saturate(None=Sign(None=@n25)), B=@n10.G)
  n27 = LinearInterpolate(A=AppendVector(A=@n9.R, B=Abs(None=@n24)), B=AppendVector(A=Abs(None=@n26), B=@n10.R), Alpha=@n11.Result)
  n28 = Constant()
  NamedRerouteDeclaration(None=StaticSwitch(
    True=Transform(None=@n21) {transformSourceType=TRANSFORMSOURCE_World, transformType=TRANSFORM_Tangent}
    False=AppendVector(A=@n27, B=@n28)
    Value=NamedRerouteUsage()
   )) {name=Normal, nodeColor=<rgba 0 0.2031 1 1>}
  n29 = FunctionInput() {bUsePreviewValueAsDefault=true, inputName=OasisMask, inputType=FunctionInput_Scalar, previewValue=<0 0 0 0>}
  NamedRerouteDeclaration(None=@n29) {name=OasisMask, nodeColor=<rgba 0 0.5781 1 1>}
  NamedRerouteDeclaration(None=LinearInterpolate(A=AppendVector(A=@n9.B, B=@n9.B), B=AppendVector(A=@n10.B, B=@n10.B), Alpha=@n11.Result)) {name=Texture, nodeColor=<rgba 0 0.2031 1 1>}
  n30 = NamedRerouteUsage()
  n31 = Abs(None=ComponentMask(None=VertexNormalWS()) {b=true})
  n32 = Multiply(A=ComponentMask(None=NamedRerouteUsage()) {b=true}) {constB=2}
  n33 = Subtract(A=Multiply(A=OneMinus(None=@n31)) {constB=5}, B=Saturate(None=Subtract(A=@n32) {constB=0.5}))
  NamedRerouteDeclaration(None=StaticSwitch(True=ComponentMask(None=@n30) {g=true}, False=Saturate(None=@n33), Value=NamedRerouteUsage())) {name=Height, nodeColor=<rgba 0.8984 1 0 1>}
  n34 = VectorParameter() {group=Mobile CliffSedimentary, parameterName=Mobile CliffSedimentary Color 1, defaultValue=<rgba 0.1875 0.0464 0.0195 1>}
  n35 = VectorParameter() {group=Mobile CliffSedimentary, parameterName=Mobile CliffSedimentary Color 2, defaultValue=<rgba 0.5208 0.1467 0.0787 1>}
  n36 = VectorParameter() {group=Mobile CliffSedimentary, parameterName=Mobile CliffSedimentary Color 3, defaultValue=<rgba 0.224 0.0553 0.0464 1>}
  n37 = NamedRerouteUsage()
  n38 = LinearInterpolate(A=@n35.RGB, B=@n36.RGB, Alpha=ComponentMask(None=@n37) {r=true})
  n39 = VectorParameter() {group=Mobile CliffSedimentary, parameterName=Mobile CliffSedimentary Color 4, defaultValue=<rgba 0.2344 0.1097 0.0317 1>}
  n40 = Add(A=Multiply(A=NamedRerouteUsage(), B=@n39.RGB))
  n41 = ComponentMask(None=@n37) {g=true}
  n42 = VectorParameter() {group=Mobile CliffSedimentary, parameterName=Mobile CliffSedimentary Color 5, defaultValue=<rgba 0.2917 0.1209 0.0547 1>}
  n43 = Multiply(A=@n41, B=ScalarParameter() {defaultValue=4, group=Mobile CliffSedimentary, parameterName=Mobile CliffSedimentary SpecklesMul}) {constB=2.7}
  n44 = ScalarParameter() {defaultValue=2.2, group=Mobile CliffSedimentary, parameterName=Mobile CliffSedimentary SpecklesSub}
  n45 = LinearInterpolate(A=LinearInterpolate(A=@n34.RGB, B=Multiply(A=@n38, B=@n40), Alpha=@n41), B=@n42.RGB, Alpha=Saturate(None=Subtract(A=@n43, B=@n44)))
  n46 = VectorParameter() {group=Mobile CliffSedimentary, parameterName=Mobile CliffSedimentary Bands, defaultValue=<rgba 0.9885 1.9564 2.6 1>}
  n47 = NamedRerouteUsage()
  n48 = Multiply(A=ComponentMask(None=NamedRerouteUsage()) {r=true}, B=NamedRerouteUsage())
  NamedRerouteDeclaration(None=LinearInterpolate(A=@n45, B=Multiply(A=@n45, B=@n46.RGB) {constB=0.5}, Alpha=Multiply(A=@n47, B=@n48))) {name=ClifflBaseColor, nodeColor=<rgba 0.5 0 1 1>}
  NamedRerouteDeclaration(None=FunctionInput() {bUsePreviewValueAsDefault=true, inputName=VariationMasks Mobile, inputType=FunctionInput_Vector4}) {name=VariationMasks Mobile, nodeColor=<rgba 1 0.7181 0.0414 1>}
  n49 = ComponentMask(None=NamedRerouteUsage()) {g=true}
  n50 = ScalarParameter() {defaultValue=0.4, group=Mobile CliffSedimentary, parameterName=Mobile CliffSedimentary Specular}
  NamedRerouteDeclaration(None=Saturate(None=Multiply(A=@n49, B=@n50))) {name=CliffSpecular, nodeColor=<rgba 0.5 0 1 1>}
  n51 = Multiply(
   A=ComponentMask(None=NamedRerouteUsage()) {g=true}
   B=ScalarParameter() {defaultValue=0.6, group=Mobile CliffSedimentary, parameterName=Mobile CliffSedimentary Roughness}
  )
  NamedRerouteDeclaration(None=Saturate(None=OneMinus(None=@n51))) {name=CliffRoughness, nodeColor=<rgba 0.1484 0 1 1>}
  NamedRerouteDeclaration(None=FunctionInput(Preview=StaticBool()) {bUsePreviewValueAsDefault=true, inputName=IsStaticMesh, inputType=FunctionInput_StaticBool, sortPriority=1}) {name=IsStaticMesh, nodeColor=<rgba 1 0.6797 0 1>}
  NamedRerouteDeclaration(None=@n5.Sediment Wave Distortion) {name=SedimentWaveDistortion, nodeColor=<rgba 1 0.9141 0 1>}
  n52 = Subtract(
   A=ComponentMask(None=@n2) {b=true}
   B=ScalarParameter() {defaultValue=500, group=Mobile CliffSedimentary, parameterName=Mobile CliffSedimentary BandOffset, sortPriority=501}
  )
  n53 = NamedRerouteUsage()
  n54 = Frac(None=Multiply(
    A=Subtract(A=@n52, B=@n53)
    B=ScalarParameter() {defaultValue=0.0006, group=Mobile CliffSedimentary, parameterName=Mobile CliffSedimentary BandSize, sortPriority=501}
   ))
  n55 = Abs(None=Add(A=Multiply(A=@n54) {constB=2}) {constB=-1})
  n56 = Multiply(
   A=Multiply(A=@n55, B=@n55)
   B=ScalarParameter() {defaultValue=6, group=Mobile CliffSedimentary, parameterName=Mobile CliffSedimentary BandContrast, sortPriority=501}
  ) {constB=4}
  n57 = ScalarParameter() {defaultValue=-0.5, group=Mobile CliffSedimentary, parameterName=Mobile CliffSedimentary BandSub, sortPriority=501}
  NamedRerouteDeclaration(None=Saturate(None=Add(A=@n56, B=@n57))) {name=BandsVariety, nodeColor=<rgba 0.5938 0 1 1>}
  Reroute()
 ]}
